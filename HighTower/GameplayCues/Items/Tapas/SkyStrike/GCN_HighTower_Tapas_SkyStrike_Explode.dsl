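; Burst gameplay cue for the HighTower Tapas Sky Strike explosion, adapting impact audio and Cascade effects to water.
; If the target is a Fort projectile, its water interaction component selects the water or ground impact sound.
; A vertical line trace around the first particle component treats any hit as water, switching that effect's float parameters from smoke to water spawning.
; Debug water/nowater print strings remain in the trace branches.
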(fn OnBurst(Cascade)
  (params (MyTarget :type Actor%20Object%20Reference) (Parameters :type Gameplay%20Cue%20Parameters%20Structure%20%28by%20ref%29) (ParticleComponents :type Array%20of%20Cascade%20Particle%20System%20Component%20Object%20References) (AudioComponents :type Array%20of%20Audio%20Component%20Object%20References) (BurstCameraShakeInstance :type Camera%20Shake%20Base%20Object%20Reference) (BurstDecalInstance :type Decal%20Actor%20Object%20Reference))
  (bind _output (Utilities|Array|Get(acopy) ParticleComponents))
  (bind _asfort_projectile_base (Utilities|Casting|CastToFortProjectileBase MyTarget)
    (:then
      (Audio|PlaySoundatLocation MyTarget (select (Water|IsinWater (Variables|/Script/FortniteGame.FortProjectileBase|GetWaterInteractionComponent _asfort_projectile_base)) "/HighTower/Audio/Tapas/Tapas_Strike_GroundImpact_Water_Cue.Tapas_Strike_GroundImpact_Water_Cue" "/HighTower/Audio/Tapas/Tapas_Strike_GroundImpact_Cue.Tapas_Strike_GroundImpact_Cue") (Transformation|GetActorLocation MyTarget)))
    (:CastFailed
      _))
  (if (> (Utilities|Array|Length ParticleComponents) 0)
    (bind (_outhit _returnvalue) (Collision|LineTraceByChannel :WorldContextObject MyTarget :Start (+ (Transformation|GetWorldLocation _output) "0, 0,200.000000") :End (+ (Transformation|GetWorldLocation _output) "0, 0,-500.000000") :TraceChannel "TraceTypeQuery12" :bTraceComplex false :DrawDebugType "None" :bIgnoreSelf true :TraceColor "(R=1.000000,G=0.000000,B=0.000000,A=1.000000)" :TraceHitColor "(R=0.000000,G=1.000000,B=0.000000,A=1.000000)" :DrawTime 3e+01))
    (if _returnvalue
      (Development|PrintString MyTarget "water" true true "(R=0.000000,G=0.660000,B=1.000000,A=1.000000)" 8.0)
      (Effects|Components|ParticleSystem|SetFloatParameter _output "SpawnSmoke" 0.0)
      (Effects|Components|ParticleSystem|SetFloatParameter _output "SpawnWater" 1.0)
      (else
        (Development|PrintString MyTarget "nowater" true true "(R=0.000000,G=0.660000,B=1.000000,A=1.000000)" 8.0)))
    (return)
    (else
      (return))))
